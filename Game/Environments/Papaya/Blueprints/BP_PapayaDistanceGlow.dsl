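; BP_PapayaDistanceGlow is an environment actor for a glow effect with adjustable size and color, shown on a static mesh component.
; Its construction script creates a dynamic material instance on the mesh's first material slot and stores it as MID.
; It then feeds the Size and Color variables into that material's "Scale" and "Color" parameters.
; BeginPlay, overlap and Tick events are empty, so any distance-based behavior is not in this graph.
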(event EventBeginPlay)

(event Collision|EventActorBeginOverlap
  (params (OtherActor :type Actor%20Object%20Reference)))

(event EventTick
  (params (DeltaSeconds :type Float%20%28single-precision%29)))

(fn ConstructionScript
  (bind _returnvalue (Rendering|Material|CreateDynamicMaterialInstance (Variables|Default|GetStaticMeshComponent0) 0 0 "None" :!class /Script/Engine.PrimitiveComponent))
  (bind _output_get (Variables|Default|SetMID _returnvalue))
  (Rendering|Material|SetScalarParameterValue _output_get "Scale" (Variables|Default|GetSize) :!class /Script/Engine.MaterialInstanceDynamic)
  (Rendering|Material|SetVectorParameterValue _output_get "Color" (Variables|Default|GetColor) :!class /Script/Engine.MaterialInstanceDynamic))
