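; BP_PapayaPoster_Control is a Blueprint that switches special event posters in the Papaya environment.
; Two custom events, commented as mesh network controlled, write scalars to the shared MPC_Countdown material parameter collection.
; Both set ShowPapayaSpecialEventPoster from their Enabled input; when enabled, the first clears ShowPapayaSpecialEventPoster2 and the second sets it.
; Disabling leaves that second parameter untouched, and BeginPlay, overlap and construction script are empty.
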(event EventBeginPlay)

(event Collision|EventActorBeginOverlap
  (params (OtherActor :type Actor%20Object%20Reference)))

(event Custom|EnableSpecialEventPoster
  (params (Enabled :type Boolean))
  ;[1 1 1] Mesh Network Controlled
  (Rendering|Material|SetScalarParameterValue "/Game/Athena/Events/Generic/CountdownTimer/Materials/MPC_Countdown.MPC_Countdown" "ShowPapayaSpecialEventPoster" (Math|Float|SelectFloat 1.0 0.0 Enabled) :!class /Script/Engine.KismetMaterialLibrary)
  (if Enabled
    (Rendering|Material|SetScalarParameterValue "/Game/Athena/Events/Generic/CountdownTimer/Materials/MPC_Countdown.MPC_Countdown" "ShowPapayaSpecialEventPoster2" :!class /Script/Engine.KismetMaterialLibrary)))

(event Custom|EnableSpecialEventPoster2
  (params (Enabled :type Boolean))
  ;[1 1 1] Mesh Network Controlled
  (Rendering|Material|SetScalarParameterValue "/Game/Athena/Events/Generic/CountdownTimer/Materials/MPC_Countdown.MPC_Countdown" "ShowPapayaSpecialEventPoster" (Math|Float|SelectFloat 1.0 0.0 Enabled) :!class /Script/Engine.KismetMaterialLibrary)
  (if Enabled
    (Rendering|Material|SetScalarParameterValue "/Game/Athena/Events/Generic/CountdownTimer/Materials/MPC_Countdown.MPC_Countdown" "ShowPapayaSpecialEventPoster2" 1.0 :!class /Script/Engine.KismetMaterialLibrary)))

(fn ConstructionScript)
