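; BP_Papaya_Screen_Shards is a Papaya environment Blueprint that switches visibility between a main screen shards component and a side screen shards component.
; The construction script shows the main shards and hides the side shards by default.
; ShowShardSideScreens and HideShardSideScreens swap the two sets, while ShowMainScreenShard and HideMainScreenShard toggle only the main one.
; BeginPlay, Tick and overlap events are empty, and the file contains no callers for the custom events.
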(event EventBeginPlay)

(event Collision|EventActorBeginOverlap
  (params (OtherActor :type Actor%20Object%20Reference)))

(event EventTick
  (params (DeltaSeconds :type Float%20%28single-precision%29)))

(event Custom|ShowShardSideScreens
  (Development|SetHiddeninGame (Variables|Default|GetSM_Papaya_SideScreen_Shards) false false)
  (Development|SetHiddeninGame (Variables|Default|GetSM_Papaya_Screen_Shards) true false))

(event Custom|HideShardSideScreens
  (Development|SetHiddeninGame (Variables|Default|GetSM_Papaya_SideScreen_Shards) true false)
  (Development|SetHiddeninGame (Variables|Default|GetSM_Papaya_Screen_Shards) false false))

(event Custom|HideMainScreenShard
  (Development|SetHiddeninGame (Variables|Default|GetSM_Papaya_Screen_Shards) true false))

(event Custom|ShowMainScreenShard
  (Development|SetHiddeninGame (Variables|Default|GetSM_Papaya_Screen_Shards) false false))

(fn ConstructionScript
  (Development|SetHiddeninGame (Variables|Default|GetSM_Papaya_Screen_Shards) false false)
  (Development|SetHiddeninGame (Variables|Default|GetSM_Papaya_SideScreen_Shards) true false))
