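; BP_Papaya_TV is a television prop Blueprint for the Papaya environment that switches between off, music stage, movie screen and color sync displays.
; Each custom event swaps the video player material on slot 0 of the TV mesh.
; The music stage and movie screen events also start the audio component, DisableTV fades it out, and color sync leaves audio untouched.
; The construction script is empty, and no event here calls these modes.
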(event Custom|DisableTV
  (Rendering|Material|SetMaterial (Variables|Default|GetS_Papaya_TV_Mesh) 0 "/Game/Environments/Papaya/Materials/M_PapayaTV_Videoplayer_Off.M_PapayaTV_Videoplayer_Off")
  (Audio|Components|Audio|FadeOut (Variables|Default|GetAudio) 1.0 0.0 "Linear" :!class /Script/Engine.AudioComponent))

(event Custom|EnableTVMusicStage
  (Rendering|Material|SetMaterial (Variables|Default|GetS_Papaya_TV_Mesh) 0 "/Game/Environments/Papaya/Materials/M_PapayaTV_Videoplayer_ShardScreen.M_PapayaTV_Videoplayer_ShardScreen")
  (Audio|Components|Audio|Play (Variables|Default|GetAudio) 0.0))

(event Custom|EnableTVMovieScreen
  (Rendering|Material|SetMaterial (Variables|Default|GetS_Papaya_TV_Mesh) 0 "/Game/Environments/Papaya/Materials/M_PapayaTV_Videoplayer_MovieScreen.M_PapayaTV_Videoplayer_MovieScreen")
  (Audio|Components|Audio|Play (Variables|Default|GetAudio) 0.0))

(event Custom|EnableTVColorSync
  (Rendering|Material|SetMaterial (Variables|Default|GetS_Papaya_TV_Mesh) 0 "/Game/Environments/Papaya/Materials/M_PapayaTV_Videoplayer_ColorSync.M_PapayaTV_Videoplayer_ColorSync"))

(fn ConstructionScript)
