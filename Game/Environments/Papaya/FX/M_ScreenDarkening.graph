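# Translucent, unlit, two-sided effect material for soft radial screen darkening, going by its name, in the Papaya environment.
# A 2D sphere gradient raised to RadialExponent is scaled by an Opacity parameter and clamped.
# The absolute camera-to-normal dot product fades that mask at grazing angles, and DepthFade softens it near intersecting geometry.
# A CameraOffset function with a negative default supplies a camera-relative world position offset.
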Material /Game/Environments/Papaya/FX/M_ScreenDarkening
graph {
 properties {blendMode=BLEND_Translucent, pixelDepthOffsetMode=PDOM_Legacy, shadingModel=MSM_Unlit, twoSided=true, usedShadingModels=Unlit}
 nodes [
  n1 = Multiply(
   A=Power(
    Base=MaterialFunctionCall!(UVs, CenterPosition, Radius) {materialFunction=/Engine/Functions/Engine_MaterialFunctions01/Gradient/SphereGradient-2D}.Result 0-1
    Exp=ScalarParameter() {defaultValue=4.4267, parameterName=RadialExponent}
   )
   B=ScalarParameter() {parameterName=Opacity}
  )
  n2 = DotProduct(A=CameraVectorWS(), B=PixelNormalWS())
  DepthFade(Opacity=Multiply(A=Clamp(None=@n1), B=Abs(None=@n2)), FadeDistance=ScalarParameter() {defaultValue=2000, parameterName=FadeDepth})
  MaterialFunctionCall(Offset Amount=ScalarParameter() {defaultValue=-100, parameterName=offset}) {materialFunction=/Engine/Functions/Engine_MaterialFunctions02/WorldPositionOffset/CameraOffset}
 ]
}
