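# Additive unlit material for a distance glow: a camera-facing, tinted radial halo that is strongest at night, half strength morning and evening, and off by day.
# The Sprite function builds the scaled billboard, which is pushed toward the camera.
# Radial falloff, scene depth fade and camera depth fade shape opacity.
# When camera fade reaches zero, the vertex offset is zeroed too, dropping the billboard expansion.
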Material /Game/Environments/Papaya/Materials/M_DistanceGlow
graph {
 properties {blendMode=BLEND_Additive, pixelDepthOffsetMode=PDOM_Legacy, shadingModel=MSM_Unlit, usedShadingModels=Unlit}
 nodes [
  n1 = ScalarParameter() {defaultValue=8000, parameterName=Scale}
  n2 = ScalarParameter() {defaultValue=0.8, parameterName=Scale_Y}
  n3 = MaterialFunctionCall!(Center Location=ObjectPositionWS(), 0-1 UVs, XY scale=AppendVector(A=@n1, B=Multiply(A=@n1, B=@n2)), OverridePivot, "Normals (Optional)") {materialFunction=&Sprite}
  n4 = WorldPosition()
  n5 = Normalize(VectorInput=Subtract(A=CameraPositionWS(), B=Add(A=@n3.World position offset, B=@n4.XYZ)))
  n6 = Divide(A=@n1)
  n7 = Multiply(A=ScalarParameter() {defaultValue=1, parameterName=CameraPush}, B=@n6)
  n8 = MaterialFunctionCall(Fade Length=Multiply(A=ScalarParameter() {defaultValue=3, parameterName=FadeLength}, B=@n1), Fade Offset=@n6, For Vertex Shader=StaticBool()) {materialFunction=&CameraDepthFade}
  n9 = Constant()
  LinearInterpolate(
   A=Constant3Vector()
   B=Add(A=@n3.World position offset, B=Multiply(A=@n5, B=@n7))
   Alpha=If(A=@n8.Result, B=@n9, "A > B"=Constant() {r=1}, "A == B"=@n9, "A < B"=@n9)
  )
  n10 = Divide(A=Distance(A=TextureCoordinate(), B=Constant2Vector() {g=0.5, r=0.5}), B=ScalarParameter() {defaultValue=0.6, parameterName=Size})
  n11 = DepthFade(Opacity=Power(Base=OneMinus(None=@n10), Exp=ScalarParameter() {defaultValue=3, parameterName=Falloff}), FadeDistance=@n1)
  n12 = ScalarParameter() {defaultValue=0.08, parameterName=Opacity}
  Multiply(
   A=Multiply(A=@n8.Result, B=Multiply(A=@n11, B=@n12))
   B=MaterialFunctionCall(Value_Morning=Constant() {r=0.5}, Value_Day=Constant(), Value_Evening=Constant() {r=0.5}, Value_Night=Constant() {r=1}) {materialFunction=&Scalar_Time_of_Day}.Result
  )
  Multiply(
   A=VectorParameter() {parameterName=Color, defaultValue=<rgba 1 0.7459 0.1808 1>}.RGB
   B=ScalarParameter() {defaultValue=4, parameterName=Brightness}
  )
  ScalarParameter() {defaultValue=2000, parameterName=FadeDistance}
 ]
}
