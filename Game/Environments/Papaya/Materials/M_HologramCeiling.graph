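# Translucent, unlit, two-sided hologram ceiling material for the Papaya environment that shows a media-player video reacting to event audio.
# The red audio-spectrum channel from a parameter collection scales a panning smoke texture's glow and vertical UV distortion, plus video brightness.
# Three BumpOffset samples at different heights split the video's RGB channels, with audio-driven separation.
# Opacity combines a spherical UV mask, view-angle falloff, depth fade and a scalar parameter.
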Material /Game/Environments/Papaya/Materials/M_HologramCeiling
graph {
 properties {blendMode=BLEND_Translucent, pixelDepthOffsetMode=PDOM_Legacy, shadingModel=MSM_Unlit, twoSided=true, usedShadingModels=Unlit}
 nodes [
  Constant() {r=0.1}
  n1 = TextureCoordinate() {uTiling=0.1}
  n2 = Multiply(
   A=MaterialFunctionCall() {materialFunction=&LinearGradient}.UGradient
   B=Constant2Vector() {g=0.2}
  )
  n3 = TextureSample(UVs=Panner(Coordinate=Add(A=@n1, B=@n2)) {speedX=-1}) {texture=&T_Inky_Smoke_Tile_Chromatic}
  n4 = CollectionParameter() {collection=&FortniteMaterialParameters, parameterName=EventAudio_Spectrum}
  n5 = Power(Base=Clamp(None=ComponentMask(None=@n4) {r=true}), Exp=ScalarParameter() {defaultValue=3, parameterName=percExponent})
  n6 = Multiply(A=@n3.RGB, B=@n5)
  n7 = Constant() {r=-2}
  n8 = LinearInterpolate(Alpha=@n5) {constA=0.5}
  n9 = LinearInterpolate(Alpha=@n5) {constA=0.02, constB=0.3}
  n10 = Add(A=TextureCoordinate(), B=Multiply(A=Multiply(A=@n3.R, B=@n9), B=Constant2Vector() {g=1}))
  n11 = Constant() {r=1}
  n12 = Multiply(A=@n11, B=Constant() {r=0.5})
  n13 = LinearInterpolate(A=ScalarParameter() {defaultValue=0.02, parameterName=rat}, B=ScalarParameter() {defaultValue=0.08, parameterName=rat2}, Alpha=@n5)
  n14 = TextureObject() {texture=/Game/Movies/Season5/Papaya_MediaPlayer_Video2}
  n15 = Constant() {r=-1}
  n16 = Multiply(A=@n11, B=Constant() {r=-1.5})
  n17 = MaterialFunctionCall(
   R=TextureSample(UVs=BumpOffset(Coordinate=@n10, Height=@n12, HeightRatioInput=@n13), Tex=@n14) {texture=/Game/UI/Foundation/Textures/LoadingScreens/PSA/VBuckScamWarning/BR04_LoadingScreen_VBucksScam}.R
   G=TextureSample(UVs=BumpOffset(Coordinate=@n10, Height=@n15, HeightRatioInput=@n13), Tex=@n14) {texture=&BR04_LoadingScreen_VBucksScam}.G
   B=TextureSample(UVs=BumpOffset(Coordinate=@n10, Height=@n16, HeightRatioInput=@n13) {heightRatio=-1.5}, Tex=@n14) {texture=&BR04_LoadingScreen_VBucksScam}.B
  ) {materialFunction=&AppendMany}
  MaterialFunctionCall(Emissive=Add(A=Desaturation(None=@n6, Fraction=@n7), B=Multiply(A=@n8, B=@n17.RGB))) {materialFunction=&MF_FX_EmissiveAdjustments}
  MaterialFunctionCall() {materialFunction=&DitherTemporalAA}
  n18 = Power(Base=SphereMask(A=TextureCoordinate(), B=Constant() {r=0.5}, Radius=Constant() {r=0.5}, Hardness=Constant()))
  n19 = Abs(None=DotProduct(A=CameraVectorWS(), B=PixelNormalWS()))
  Multiply(
   A=DepthFade(Opacity=Multiply(A=@n18, B=@n19), FadeDistance=ScalarParameter() {defaultValue=800, parameterName=DepthFade})
   B=ScalarParameter() {parameterName=Opacity}
  )
  TextureObject() {texture=&BR04_LoadingScreen_VBucksScam}
  Time()
 ]
}
