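# Material for Papaya TV screens, layering an emissive media-player video, or a synced gadget color, over the TV's base textures.
# Static switches choose scaled UV1, planar projection across wall bounds, or wrapped radial projection for towers.
# Video passes through inverse tonemapping and shows only where red or green vertex color marks it.
# Optional masking uses a noisy ramp driven by a collection parameter, adding cyan glow where the mask falls off.
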Material /Game/Environments/Papaya/Materials/M_PapayaTV_Videoplayer
graph {
 properties {bUseMaterialAttributes=true, bUsedWithNanite=true, pixelDepthOffsetMode=PDOM_Legacy, twoSided=true, usedShadingModels=Default Lit}
 nodes [
  ActorPositionWS()
  n1 = ComponentMask(None=VectorParameter() {parameterName=BoundSize, defaultValue=<rgba 0 0 0 1>}.RGB) {b=true, r=true}
  Multiply(A=@n1, B=ScalarParameter() {defaultValue=1, parameterName=BiasVideoSize})
  n2 = TextureSample() {samplerType=SAMPLERTYPE_Masks, texture=/Game/Environments/Papaya/Textures/T_Papaya_TV_S}
  n3 = SetMaterialAttributes!(
   MaterialAttributes
   Base Color=TextureSample() {texture=/Game/Environments/Papaya/Textures/T_Papaya_TV_D}.RGB
   Metallic=@n2.G
   Specular=@n2.G
   Roughness=@n2.G
   Normal=TextureSample() {samplerType=SAMPLERTYPE_Normal, texture=/Game/Environments/Papaya/Textures/T_Papaya_TV_N}.RGB
  ) {attributeSetTypes [
    #1
    #2
    #3
    #4
    #5
   ]}
  n4 = Multiply(
   A=MaterialFunctionCall!(
    Media Player TexCoord
    Media Color Code Coordinate
    Media Texture Resolution
    Base Color Parameter
    Base Color Mix
    Media Color Code Mix
    Intermission Color Parameter
    Intermission Color Mix
    Desaturation
   ) {materialFunction=/Game/Environments/Papaya/Materials/MaterialFunctions/MF_Papaya_GadgetColor}.Result
   B=ScalarParameter() {defaultValue=2, parameterName=Color Multiplier}
  )
  n5 = Subtract(A=WorldPosition().XYZ, B=VectorParameter() {parameterName=BoundOrigin, defaultValue=<rgba 0 0 0 1>}.RGB)
  n6 = ComponentMask(None=@n5) {g=true, r=true}
  n7 = DotProduct(A=@n5, B=Transform(None=Constant3Vector() {constant=<rgba 1 0 0 0>}) {transformSourceType=TRANSFORMSOURCE_Local})
  n8 = DotProduct(A=@n5, B=Transform(None=Constant3Vector() {constant=<rgba 0 0 -1 0>}) {transformSourceType=TRANSFORMSOURCE_Local})
  n9 = Subtract(A=Divide(A=AppendVector(A=@n7, B=@n8), B=@n1), B=Constant2Vector() {g=-0.5, r=-0.5})
  n10 = Multiply(
   A=AppendVector(
    A=MaterialFunctionCall(Vector or UVs=@n6) {materialFunction=&VectorToRadialValue}.Vector Converted to Angle
    B=ComponentMask(None=@n9) {g=true}
   )
   B=Constant2Vector() {g=1, r=2}
  )
  n11 = ScalarParameter() {defaultValue=1.11, parameterName=VideoScaleX, sliderMin=1}
  n12 = ScalarParameter() {defaultValue=1, parameterName=VideoScaleY}
  n13 = StaticSwitchParameter(
   True=StaticSwitchParameter(True=Frac(None=@n10), False=@n9) {parameterName=isTVTower}
   False=MaterialFunctionCall(UVs=TextureCoordinate() {coordinateIndex=1}, Texture Scale=AppendVector(A=@n11, B=@n12)) {materialFunction=&ScaleUVsByCenter}.UVs
  ) {parameterName=isTVWall}
  n14 = VertexColor()
  n15 = Add(A=@n14.R, B=@n14.R)
  n16 = Add(A=@n14.G, B=@n14.G)
  n17 = LinearInterpolate(
   A=Constant3Vector()
   B=MaterialFunctionCall!(UVs=@n13, Texture=TextureObjectParameter() {parameterName=MediaPlayerSample, texture=/Game/Movies/Season5/Papaya_MediaPlayer_Video2}) {materialFunction=/Game/Movies/InversePostProcess/MF_invertFNtonemap}.Result
   Alpha=Saturate(None=Add(A=@n15, B=@n16))
  )
  n18 = StaticSwitchParameter(
   True=SetMaterialAttributes!(MaterialAttributes=@n3, Emissive Color=StaticSwitchParameter(True=@n4, False=@n17) {parameterName=IsColorSync}) {attributeSetTypes [#6]}
   False=@n3
  ) {defaultValue=true, parameterName=VideoEnabled}
  n19 = Constant()
  n20 = ComponentMask(None=Multiply(A=@n13) {constB=30}) {g=true}
  n21 = LinearInterpolate(Alpha=TextureSample(UVs=AppendVector(A=@n19, B=@n20)) {samplerType=SAMPLERTYPE_LinearGrayscale, texture=/Game/Global/Textures/Default/Noise/EqualizerComponents_v2_Noise}.R) {constA=-0.03, constB=0.03}
  n22 = TextureSample(UVs=@n13) {samplerType=SAMPLERTYPE_LinearGrayscale, texture=/Game/Global/Textures/Default/Masks_and_Gradients/EqualizerComponents_v2_Ramp}
  n23 = CollectionParameter() {
   collection=&FortniteMaterialParameters
   parameterName=PapayaTVTowerMaskParameter
  }
  n24 = LinearInterpolate(Alpha=@n23) {constA=-0.025}
  n25 = Saturate(None=MaterialFunctionCall(Input=Add(A=@n21, B=@n22.R), Input Low=@n24, Input High=Add(A=@n24) {constB=0.05}, Target Low=Constant(), Target High=Constant() {r=1}) {materialFunction=&RemapValueRange}.Result)
  n26 = GetMaterialAttributes(None=SetMaterialAttributes!(MaterialAttributes=@n18, Opacity Mask=@n25) {attributeSetTypes [#7]}) {attributeGetTypes [#6]}
  n27 = Multiply(A=Constant3Vector() {constant=<rgba 0.2223 2.2758 2.5 0>}, B=OneMinus(None=@n25))
  StaticSwitch(
   True=SetMaterialAttributes!(MaterialAttributes=@n26.MaterialAttributes, Emissive Color=Add(A=@n26.Emissive Color, B=@n27)) {attributeSetTypes [#6]}
   False=@n18
   Value=StaticBoolParameter() {parameterName=IsMasked}
  )
  Subtract(A=WorldPosition() {worldPositionShaderOffset=WPT_ExcludeAllShaderOffsets}.XYZ, B=ObjectPositionWS())
 ]
}
