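# Lit material for a Papaya info board prop that shows the Papaya minimap with a marker at the board's own location.
# Object world position is normalized by terrain size, swizzled and offset to place the tinted marker over the offset, scaled map UVs.
# The info board mask's alpha, boosted and saturated, blacks out the diffuse texture and reveals the map and marker layer in that area.
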Material /Game/Environments/Papaya/Materials/M_Papaya_InfoBoard
graph {
 properties {bUsedWithNanite=true, pixelDepthOffsetMode=PDOM_Legacy, usedShadingModels=Default Lit}
 nodes [
  n1 = TextureSample() {samplerType=SAMPLERTYPE_Masks, texture=/Game/Environments/Sets/Neo/Props/Textures/T_NeoTilted_InfoBoard_01_S}
  n2 = TextureCoordinate()
  n3 = ComponentMask(None=VectorParameter() {parameterName=UVOffset, defaultValue=<rgba -0.2714 -0.1714 0 1>}.RGB) {g=true, r=true}
  n4 = Clamp(None=MaterialFunctionCall(UVs=Add(A=@n2, B=@n3), Texture Scale=ScalarParameter() {defaultValue=0.6881, parameterName=Scalemap}) {materialFunction=&ScaleUVsByCenter}.UVs)
  n5 = Multiply(
   A=TextureSample(UVs=@n4) {texture=/Game/Athena/Apollo/Maps/Special/Papaya/UI/MiniMapPapaya}.RGB
   B=LinearInterpolate(A=ScalarParameter() {parameterName=Brightness_1}, B=ScalarParameter() {defaultValue=1, parameterName=Brightness_2}, Alpha=@n1.A)
  )
  n6 = VectorParameter() {parameterName=ColorTarget, defaultValue=<rgba 1 0 0 1>}
  n7 = Constant() {r=-0.5}
  n8 = ScalarParameter() {defaultValue=0.47, parameterName=offsetYuv}
  n9 = ComponentMask(None=VectorParameter() {parameterName=SizeOffTerrain, defaultValue=<rgba 49786.3711 49786.3711 65024 1>}.RGB) {g=true, r=true}
  n10 = ObjectPositionWS()
  n11 = Divide(A=Add(A=@n9, B=ComponentMask(None=@n10) {g=true, r=true}), B=Multiply(A=@n9) {constB=2})
  n12 = Add(
   A=Subtract(A=@n4, B=AppendVector(A=@n7, B=@n8))
   B=Multiply(
    A=MaterialFunctionCall(XY=@n11) {materialFunction=&Swizzle}.YX
    B=Constant2Vector() {g=1, r=-1}
   )
  )
  n13 = MaterialFunctionCall(UVs=@n12, Texture Scale=ScalarParameter() {defaultValue=0.0753, parameterName=Target Scale}) {materialFunction=&ScaleUVsByCenter}
  n14 = Multiply(A=TextureSample(UVs=@n13.UVs) {samplerType=SAMPLERTYPE_Masks, texture=/Game/Environments/Papaya/Textures/T_Marker}.RGB, B=@n13.0-1 mask)
  Add(A=@n5, B=Multiply(A=@n6.RGB, B=@n14))
  Add(A=@n12, B=Constant2Vector() {g=0.5})
  AppendVector(A=@n11, B=Constant())
  ComponentMask(None=VectorParameter() {defaultValue=<rgba -0.5 0.5 0 1>}.RGB) {g=true, r=true}
  Constant2Vector() {g=0.5, r=-0.5}
  n15 = Saturate(None=Multiply(A=@n1.A, B=ScalarParameter() {defaultValue=8, parameterName=Opacity}))
  LinearInterpolate(A=TextureSample() {texture=/Game/Environments/Sets/Neo/Props/Textures/T_NeoTilted_InfoBoard_01_D}.RGB, B=Constant3Vector(), Alpha=@n15)
  LinearInterpolate(B=LinearInterpolate(A=@n5, B=@n6.RGB, Alpha=@n14), Alpha=@n15)
  Multiply(A=@n2, B=ComponentMask(None=VectorParameter() {parameterName=UVScale, defaultValue=<rgba 2 2 0 1>}.RGB) {g=true, r=true})
  TextureSample() {samplerType=SAMPLERTYPE_Normal, texture=/Game/Environments/Sets/Neo/Props/Textures/T_NeoTilted_InfoBoard_01_N}
 ]
}
