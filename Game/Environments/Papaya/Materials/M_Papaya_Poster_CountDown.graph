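# Papaya marquee poster material that can show event posters, a countdown or a redirect map with animated arrow, driven by countdown parameter collections.
# Countdown functions add digits and DHMS labels that blink in phase three; switches select the second countdown or, by default, a static poster.
# UV1 blends this screen onto the textured marquee.
# An optional hologram branch adds scanlines and noise-based opacity, yawing toward the camera through world position offset.
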Material /Game/Environments/Papaya/Materials/M_Papaya_Poster_CountDown
graph {
 properties {bUseMaterialAttributes=true, bUsedWithNanite=true, pixelDepthOffsetMode=PDOM_Legacy, usedShadingModels=Default Lit}
 nodes [
  n1 = CollectionParameter() {collection=/Game/Athena/Events/Generic/CountdownTimer/Materials/MPC_Countdown, parameterName=ShowPapayaRedirectMovie}
  n2 = ComponentMask(None=CollectionParameter() {collection=&MPC_Countdown, parameterName=ShowPapayaRedirectShard}) {r=true}
  Add(A=@n1, B=@n2)
  n3 = ScalarParameter() {defaultValue=0.5, parameterName=ScaleRedirectArrow}
  AppendVector(A=@n3, B=@n3)
  n4 = Multiply(A=@n3) {constB=-1}
  AppendVector(A=@n4, B=@n3)
  ComponentMask() {g=true, r=true}
  Constant() {r=1}
  Constant() {r=1}
  Constant() {r=1}
  n5 = TextureCoordinate()
  n6 = MaterialFunctionCall(Float2=@n5) {materialFunction=&BreakOutFloat2Components}
  n7 = Multiply(A=@n6.R) {constB=0.5}
  LinearInterpolate(
   A=AppendVector(A=@n7, B=@n6.G)
   B=AppendVector(A=Subtract(A=@n7) {constB=0.5}, B=@n6.G)
   Alpha=ComponentMask(None=CollectionParameter() {collection=&MPC_Countdown, parameterName=ShowPapayaSpecialEventPoster2}) {r=true}
  )
  MaterialFunctionCall!(UVs, UVPadding, "NumberofDigits(int)") {materialFunction=/Game/Packages/Fortress_SharedMaterials/Base_Material_Functions/BaseCharacterShaderFunctions/MF_CountdownTimer}
  MaterialFunctionCall!(WorldPosition=WorldPosition().XYZ, PivotLocation, WS Normals, Custom Object Basis 1, Custom Object Basis 2, Custom Object Basis 3) {materialFunction=/Engine/Functions/Engine_MaterialFunctions02/WorldPositionOffset/AlignMeshToTheCamera}
  n8 = VectorParameter() {parameterName=OffsetUV_RedirectArrow, defaultValue=<rgba -0.181 -0.9746 0 1>}
  n9 = VectorParameter() {parameterName=OffsetUV_RedirectArrowShard, defaultValue=<rgba 0 0.9429 0 1>}
  n10 = CollectionParameter() {collection=&MPC_Countdown, parameterName=ShowPapayaRedirectShard}
  n11 = Add(
   A=LinearInterpolate(A=ComponentMask(None=@n8.RGB) {g=true, r=true}, B=ComponentMask(None=@n9.RGB) {g=true, r=true}, Alpha=@n10)
   B=MaterialFunctionCall(UVs=TextureCoordinate() {vTiling=2}, Texture Scale=LinearInterpolate(A=@n3, B=@n4, Alpha=@n10)) {materialFunction=&ScaleUVsByCenter}.UVs
  )
  n12 = Add(A=Multiply(A=Time(), B=ScalarParameter() {defaultValue=2, parameterName=SpeedArrowAnim}), B=ComponentMask(None=@n11) {r=true})
  n13 = ScalarParameter() {defaultValue=2.8544, parameterName=FrequencyArrowAnim}
  n14 = Sine(None=Frac(None=Divide(A=@n12, B=@n13)))
  Multiply(A=@n14)
  n15 = StaticSwitchParameter(
   True=CollectionParameter() {collection=/Game/Athena/Environments/Blueprints/Events/MPC_Countdown_2nd, parameterName=ShowCountdown}
   False=CollectionParameter() {collection=&MPC_Countdown, parameterName=ShowCountdown}
  ) {parameterName=Is_2ndCountdown}
  Multiply(
   A=@n15
   B=CollectionParameter() {collection=&MPC_Countdown, parameterName=ShowCountdownPhase2}
  )
  n16 = Add(
   A=TextureCoordinate()
   B=AppendVector(
    A=ScalarParameter() {group=Numbers, parameterName=u_NumberPaddingOffset}
    B=ScalarParameter() {defaultValue=-0.25, group=Numbers, parameterName=v_NumberPaddingOffset}
   )
  )
  n17 = AppendVector(
   A=ScalarParameter() {defaultValue=0.9, group=Numbers, parameterName=u_NumberPaddingScale}
   B=ScalarParameter() {defaultValue=0.15, group=Numbers, parameterName=v_NumberPaddingScale}
  )
  n18 = ScalarParameter() {defaultValue=11, parameterName=NumberofDigits}
  n19 = MaterialFunctionCall!(UVs=@n16, UVPadding=@n17, "NumberofDigits(int)"=@n18) {materialFunction=/Game/Packages/Fortress_SharedMaterials/Base_Material_Functions/BaseCharacterShaderFunctions/MF_CountdownTimer_2nd}
  n20 = MaterialFunctionCall(UVs=@n16, UVPadding=@n17, "NumberofDigits(int)"=@n18) {materialFunction=&MF_CountdownTimer}
  n21 = StaticSwitchParameter(True=@n19.NumbersGreyscale, False=@n20.NumbersGreyscale) {parameterName=Is_2ndCountdown}
  OneMinus(None=@n21)
  StaticSwitchParameter(True=@n5, False=TextureCoordinate()) {parameterName=IsHologram_Edge}
  n22 = ComponentMask(None=TransformPosition(None=WorldPosition() {worldPositionShaderOffset=WPT_ExcludeAllShaderOffsets}.XYZ) {transformSourceType=TRANSFORMPOSSOURCE_World}) {b=true, r=true}
  n23 = ComponentMask(None=@n22) {g=true}
  n24 = ScalarParameter() {defaultValue=6.4334, parameterName=LineDist}
  n25 = Multiply(A=Time(), B=ScalarParameter() {defaultValue=1, parameterName=SpeedLines})
  n26 = Sine(None=Add(A=Divide(A=@n23, B=@n24), B=Frac(None=@n25)))
  n27 = Saturate(None=Ceil(None=Subtract(A=@n26) {constB=0.75}))
  n28 = TextureSample(UVs=@n5) {desc=Replaceme with Party royale, texture=/Game/Environments/Papaya/Textures/T_Papaya_Poster_Royale}
  n29 = TextureSampleParameter2D(UVs=@n5) {parameterName=Poster, texture=/Game/Environments/Papaya/Textures/T_Papaya_Poster_Next}
  n30 = Constant() {r=1}
  n31 = StaticSwitchParameter(True=@n30, False=@n15) {parameterName=Debug}
  n32 = TextureSampleParameter2D(UVs=@n5) {group=0__SPECIALEVENTTEXUTRE__0, parameterName=SpecialEventPoster, texture=/Game/Environments/Papaya/Textures/T_Papaya_Poster_Dom_n_Anders}
  n33 = TextureCoordinate()
  n34 = ComponentMask(None=VectorParameter() {parameterName=ScaleUV_Redirectmap, defaultValue=<rgba 2.0806 0.6952 0 1>}.RGB) {g=true, r=true}
  n35 = VectorParameter() {parameterName=OffsetUV_Redirectmap, defaultValue=<rgba 0.0095 0 0 1>}
  n36 = TextureSample(UVs=Add(
    A=MaterialFunctionCall(UVs=@n33, Texture Scale=@n34) {materialFunction=&ScaleUVsByCenter}.UVs
    B=ComponentMask(None=@n35.RGB) {g=true, r=true}
   )) {texture=/Game/Environments/Papaya/Textures/RedirectMap}
  n37 = ScalarParameter() {defaultValue=0.4665, parameterName=ScaleRedirectTarget}
  n38 = ComponentMask(None=VectorParameter() {parameterName=OffsetUV_RedirectTarget, defaultValue=<rgba 0.4571 0.0381 0 1>}.RGB) {g=true, r=true}
  n39 = ComponentMask(None=VectorParameter() {parameterName=OffsetUV_RedirectTargetShard, defaultValue=<rgba -0.7238 -0.0667 0 1>}.RGB) {g=true, r=true}
  n40 = TextureSample(UVs=Add(
    A=MaterialFunctionCall(UVs=@n33, Texture Scale=@n37) {materialFunction=&ScaleUVsByCenter}.UVs
    B=LinearInterpolate(A=@n38, B=@n39, Alpha=@n10)
   )) {texture=/Game/Environments/Papaya/Textures/RedirectMapTarget}
  n41 = ScalarParameter() {defaultValue=-19.1085, parameterName=m_high}
  n42 = ScalarParameter() {defaultValue=1, parameterName=m_Low}
  n43 = VertexColor()
  n44 = Multiply(A=@n40.A, B=Saturate(None=LinearInterpolate(A=@n41, B=@n42, Alpha=@n43.R)))
  n45 = TextureSample(UVs=@n11) {texture=/Game/Environments/Papaya/Textures/RedirectMapArrow}
  n46 = Ceil(None=@n14)
  n47 = LinearInterpolate(
   A=LinearInterpolate(
    A=LinearInterpolate(A=@n28.RGB, B=@n29.RGB, Alpha=@n31)
    B=@n32.RGB
    Alpha=CollectionParameter() {collection=&MPC_Countdown, parameterName=ShowPapayaSpecialEventPoster}
   )
   B=LinearInterpolate(A=LinearInterpolate(A=@n36.RGB, B=@n40.RGB, Alpha=@n44), B=@n45.RGB, Alpha=Multiply(A=@n46, B=@n45.A))
   Alpha=Saturate(None=StaticSwitchParameter(True=@n1, False=@n2) {parameterName=Is_2ndCountdown})
  )
  n48 = ScalarParameter() {group=Numbers, parameterName=u_DHMSPos}
  n49 = ScalarParameter() {defaultValue=-0.6, group=Numbers, parameterName=v_DHMSPos}
  n50 = MaterialFunctionCall(
   UVs=Add(A=AppendVector(A=@n48, B=@n49), B=StaticSwitchParameter(True=@n19.UVS, False=@n20.UVS) {parameterName=Is_2ndCountdown})
   Texture Scale=AppendVector(
    A=ScalarParameter() {defaultValue=1.0383, group=Numbers, parameterName=u_DHMSScale}
    B=ScalarParameter() {defaultValue=0.419, group=Numbers, parameterName=v_DHMSScale}
   )
  ) {materialFunction=&ScaleUVsByCenter}
  n51 = Add(
   A=Multiply(
    A=ScalarParameter() {defaultValue=1, parameterName=DHMS_Brightness}
    B=TextureSampleParameter2D(UVs=@n50.UVs) {parameterName=DHMS, samplerType=SAMPLERTYPE_Masks, texture=/Game/Athena/Playlists/Barrier/Textures/Countdown_DHMS}.R
   )
   B=@n21
  )
  n52 = Multiply(A=Time(), B=ScalarParameter() {defaultValue=1, parameterName=Blinkspeed})
  n53 = LinearInterpolate(A=@n51, B=Constant(), Alpha=Ceil(None=Sine(None=@n52)))
  n54 = StaticSwitchParameter(
   True=@n30
   False=StaticSwitchParameter(
    True=CollectionParameter() {collection=&MPC_Countdown_2nd, parameterName=ShowCountdownPhase3}
    False=CollectionParameter() {collection=&MPC_Countdown, parameterName=ShowCountdownPhase3}
   ) {parameterName=Is_2ndCountdown}
  ) {parameterName=Debug2}
  n55 = Add(A=@n47, B=LinearInterpolate(A=Constant(), B=LinearInterpolate(A=@n51, B=@n53, Alpha=@n54), Alpha=@n31))
  n56 = Multiply(
   A=LinearInterpolate(B=ScalarParameter() {defaultValue=1.2, parameterName=BrightnessLines}, Alpha=@n27) {constA=1, constB=1.5}
   B=StaticSwitchParameter(True=@n47, False=@n55) {parameterName=IsHologram_Edge}
  )
  n57 = ComponentMask(None=TextureCoordinate() {coordinateIndex=1}) {g=true}
  n58 = ScalarParameter() {defaultValue=189.3222, parameterName=SizeBlocks}
  n59 = TextureSample(UVs=Panner(Coordinate=Divide(A=@n22, B=@n58)) {bFractionalPart=true, speedX=0.02, speedY=-0.04}) {samplerType=SAMPLERTYPE_LinearColor, texture=/Game/Packages/Fortress_Sky/Materials/Noise32x322_nearestNeighboor}
  n60 = ScalarParameter() {parameterName=C}
  n61 = Multiply(A=@n27, B=@n57)
  n62 = ScalarParameter() {defaultValue=2, parameterName=LineOpacity}
  n63 = Saturate(None=Add(
    A=MaterialFunctionCall(Transition Phase=@n57, Height Texture=@n59.R, Contrast=@n60) {materialFunction=&HeightLerp}.Alpha
    B=Multiply(A=@n61, B=@n62)
   ))
  n64 = Constant3Vector() {constant=<rgba 0 0 1 0>}
  n65 = ObjectPositionWS()
  n66 = CameraPositionWS()
  n67 = Normalize(VectorInput=Multiply(A=Subtract(A=@n65, B=@n66), B=Constant3Vector() {constant=<rgba 1 1 0 0>}))
  n68 = DotProduct(A=Transform(None=Constant3Vector() {constant=<rgba 0 1 0 0>}) {transformSourceType=TRANSFORMSOURCE_Local}, B=@n67)
  n69 = Constant3Vector() {constant=<rgba 1 0 0 0>}
  n70 = MaterialFunctionCall!(Normalized V1=Normalize(VectorInput=Transform(None=@n69) {transformSourceType=TRANSFORMSOURCE_Local}), Normalized V2=@n67) {materialFunction=/Engine/Functions/Engine_MaterialFunctions02/Math/AngleBetweenVectors}
  n71 = Add(A=Multiply(A=Sign(None=@n68), B=Divide(A=@n70.Angle In Degrees) {constB=360})) {constB=0.25}
  n72 = ObjectPositionWS()
  n73 = WorldPosition()
  n74 = TextureSample() {texture=/Game/Environments/Papaya/Textures/T_Papaya_Marquee_Posters_D}
  n75 = TextureSample() {samplerType=SAMPLERTYPE_Masks, texture=/Game/Environments/Papaya/Textures/T_Papaya_Marquee_Posters_S}
  n76 = Multiply(
   A=Desaturation(None=TextureSample() {samplerType=SAMPLERTYPE_Masks, texture=/Game/Environments/Papaya/Textures/T_Papaya_Posters01_E}.RGB)
   B=VectorParameter() {parameterName=EmissiveColor, defaultValue=<rgba 3 3 3 1>}.RGB
  )
  n77 = TextureSample() {samplerType=SAMPLERTYPE_Normal, texture=/Game/Environments/Papaya/Textures/T_Papaya_Marquee_Posters_N}
  n78 = ScalarParameter() {defaultValue=0.3, parameterName=PosterEmissive}
  n79 = Constant()
  n80 = SetMaterialAttributes!(MaterialAttributes, Base Color=@n55, Emissive Color=Multiply(A=@n55, B=@n78), Roughness=@n79) {attributeSetTypes [#1, #2, #3]}
  n81 = StaticSwitchParameter(
   True=SetMaterialAttributes(Base Color=@n28.RGB, Emissive Color=Multiply(A=@n28.RGB, B=@n78), Roughness=@n79) {attributeSetTypes [#1, #2, #3]}
   False=@n80
  ) {defaultValue=true, parameterName=DisableCountdownAndJustUsePoster}
  n82 = Floor(None=ComponentMask(None=TextureCoordinate() {coordinateIndex=1}) {r=true})
  StaticSwitchParameter(
   True=SetMaterialAttributes!(
    MaterialAttributes=SetMaterialAttributes!(MaterialAttributes, Emissive Color=@n56, Opacity=@n63) {attributeSetTypes [#2, #4]}
    World Position Offset=RotateAboutAxis(NormalizedRotationAxis=@n64, RotationAngle=@n71, PivotPoint=@n72, Position=@n73.XYZ)
   ) {attributeSetTypes [#5]}
   False=BlendMaterialAttributes(
    A=MakeMaterialAttributes!(
     BaseColor=@n74.RGB
     Metallic=@n75.G
     Specular=@n75.G
     Roughness=@n75.G
     Anisotropy
     EmissiveColor=@n76
     Opacity
     OpacityMask
     Normal=@n77.RGB
     Tangent
     WorldPositionOffset
     SubsurfaceColor
     ClearCoat
     ClearCoatRoughness
     AmbientOcclusion
     Refraction
     CustomizedUVs_0
     CustomizedUVs_1
     CustomizedUVs_2
     CustomizedUVs_3
     CustomizedUVs_4
     CustomizedUVs_5
     CustomizedUVs_6
     CustomizedUVs_7
     PixelDepthOffset
     ShadingModel
     Displacement
    )
    B=StaticSwitchParameter(True=@n80, False=@n81) {parameterName=Is_2ndCountdown}
    Alpha=Saturate(None=@n82)
   )
  ) {parameterName=IsHologram}
 ]
}
