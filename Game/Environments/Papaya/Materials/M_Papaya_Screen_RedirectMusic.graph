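# Two-sided lit material for the Papaya environment, named as a music-redirect screen, that composites a redirect map, target and animated arrow.
# Each texture has its own UV scale and offset parameters, with a static switch choosing UV0 or UV1.
# The target blends over the map using its own color, then the arrow blends by alpha times a time-scrolled, Ceil-stepped sine along U.
# The graph shows no connection from this blend to a material output.
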Material /Game/Environments/Papaya/Materials/M_Papaya_Screen_RedirectMusic
graph {
 properties {pixelDepthOffsetMode=PDOM_Legacy, twoSided=true, usedShadingModels=Default Lit}
 nodes [
  Constant()
  Constant()
  n1 = StaticSwitchParameter(True=TextureCoordinate(), False=TextureCoordinate() {coordinateIndex=1}) {parameterName=UseUV0}
  n2 = VectorParameter() {parameterName=ScaleUV_Redirectmap, defaultValue=<rgba 1 1 0 1>}
  n3 = Add(
   A=MaterialFunctionCall(UVs=@n1, Texture Scale=ComponentMask(None=@n2.RGB) {g=true, r=true}) {materialFunction=&ScaleUVsByCenter}.UVs
   B=ComponentMask(None=VectorParameter() {parameterName=OffsetUV_Redirectmap, defaultValue=<rgba 0.0095 0 0 1>}.RGB) {g=true, r=true}
  )
  n4 = ComponentMask(None=VectorParameter() {parameterName=ScaleUV_RedirectTarget, defaultValue=<rgba 0.432 0.9048 0 1>}.RGB) {g=true, r=true}
  n5 = VectorParameter() {parameterName=OffsetUV_RedirectTarget, defaultValue=<rgba -0.3714 -0.0667 0 1>}
  n6 = TextureSample(UVs=Add(
    A=MaterialFunctionCall(UVs=@n1, Texture Scale=@n4) {materialFunction=&ScaleUVsByCenter}.UVs
    B=ComponentMask(None=@n5.RGB) {g=true, r=true}
   )) {texture=/Game/Environments/Papaya/Textures/RedirectMapTarget}
  n7 = VectorParameter() {parameterName=ScaleUV_RedirectArrow, defaultValue=<rgba -0.2 -0.3429 0 1>}
  n8 = Add(
   A=ComponentMask(None=VectorParameter() {parameterName=OffsetUV_RedirectArrow, defaultValue=<rgba -0.0667 -0.0762 0 1>}.RGB) {g=true, r=true}
   B=MaterialFunctionCall(UVs=@n1, Texture Scale=ComponentMask(None=@n7.RGB) {g=true, r=true}) {materialFunction=&ScaleUVsByCenter}.UVs
  )
  n9 = TextureSample(UVs=@n8) {texture=/Game/Environments/Papaya/Textures/RedirectMapArrow}
  n10 = Add(A=Multiply(A=Time(), B=ScalarParameter() {defaultValue=2, parameterName=SpeedArrowAnim}), B=ComponentMask(None=@n8) {r=true})
  n11 = ScalarParameter() {defaultValue=2.8544, parameterName=FrequencyArrowAnim}
  n12 = Sine(None=Frac(None=Divide(A=@n10, B=@n11)))
  LinearInterpolate(
   A=LinearInterpolate(A=TextureSample(UVs=@n3) {texture=/Game/Environments/Papaya/Textures/RedirectMap}.RGB, B=@n6.RGB, Alpha=@n6.RGB)
   B=@n9.RGB
   Alpha=Multiply(A=Ceil(None=@n12), B=@n9.A)
  )
  Multiply(A=@n12)
 ]
}
