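# Material function that warps UVs into an animated, fractured kaleidoscope, outputting the coordinates and an optional texture sampled with them.
# The fracture ID texture's red channel offsets each shard's rotation angle and picks spin direction around 0.5, with time and Rotation Speed driving motion.
# Rotated UVs are scaled about the pivot by inverse Aspect on both axes.
# A separate procedural fold-and-rotate chain using Screen Aspect reaches no output.
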MaterialFunction /Game/Environments/Papaya/Materials/MaterialFunctions/MF_Papaya_Kaleidoscope
graph {
 properties {bExposeToLibrary=true}
 nodes [
  n1 = TextureCoordinate()
  n2 = ScalarParameter() {defaultValue=1.5, group=MF_Papaya_Kaleidoscope, parameterName=Screen Aspect}
  n3 = MaterialFunctionCall(X=@n2, Y=Constant() {r=1}) {materialFunction=&MakeFloat2}
  n4 = Multiply(A=Subtract(A=@n2)) {constB=-0.5}
  n5 = Constant()
  n6 = Constant() {r=0.01}
  n7 = MaterialFunctionCall(
   UVs=Add(
    A=Multiply(A=@n1, B=@n3.Result)
    B=MaterialFunctionCall(X=@n4, Y=@n5) {materialFunction=&MakeFloat2}.Result
   )
   Rotation Center=Constant2Vector() {g=0.5, r=0.5}
   "Rotation Angle (0-1)"=Multiply(A=Time(), B=@n6)
  ) {materialFunction=&CustomRotator}
  n8 = MaterialFunctionCall(Float2=Abs(None=Subtract(A=@n7.Rotated Values) {constB=0.5})) {materialFunction=&BreakOutFloat2Components}
  n9 = Add(A=@n8.R, B=@n8.R)
  n10 = MaterialFunctionCall(X=@n8.R, Y=@n8.R) {materialFunction=&MakeFloat2}
  n11 = Constant() {r=0.5}
  n12 = MaterialFunctionCall(X=Subtract(A=@n11, B=@n8.R), Y=Subtract(A=@n11, B=@n8.G)) {materialFunction=&MakeFloat2}
  n13 = Time()
  n14 = MaterialFunctionCall(
   UVs=MaterialFunctionCall(
    UVs=If(A=@n9, "A > B"=@n10.Result, "A < B"=@n12.Result) {constB=0.5}
    Rotation Center=Constant2Vector() {g=0.5, r=0.5}
    "Rotation Angle (0-1)"=Multiply(A=@n13, B=@n6)
   ) {materialFunction=&CustomRotator}.Rotated Values
   Rotation Center=Constant2Vector()
   "Rotation Angle (0-1)"=Multiply(A=Time(), B=@n6)
  ) {materialFunction=&CustomRotator}
  n15 = MaterialFunctionCall(
   X=MaterialFunctionCall(X=Divide(B=@n2)) {materialFunction=&MakeFloat2}.Result
   Y=Constant() {r=1}
  ) {materialFunction=&MakeFloat2}
  Frac(None=Abs(None=Multiply(A=@n14.Rotated Values, B=@n15.Result)))
  n16 = FunctionInput(Preview=TextureCoordinate()) {
   bUsePreviewValueAsDefault=true
   description=Origianl texture coordinate to start from
   inputName=TexCoord
   inputType=FunctionInput_Vector2
   sortPriority=10
  }
  n17 = FunctionInput() {bUsePreviewValueAsDefault=true, inputName=Rotation Pivot, inputType=FunctionInput_Vector2, sortPriority=15, previewValue=<0.5 0.5 0 1>}
  n18 = FunctionInput(Preview=TextureObject() {samplerType=SAMPLERTYPE_LinearGrayscale, texture=/Game/Environments/Papaya/Textures/shatter_kaleidoscope_id}) {
   bUsePreviewValueAsDefault=true
   description=RG Texture used to replace TexCoord
   inputName=Fracture ID Texture
   inputType=FunctionInput_Texture2D
  }
  n19 = TextureSample(UVs=@n16, Tex=@n18) {samplerType=SAMPLERTYPE_LinearGrayscale, texture=&shatter_kaleidoscope_id}
  n20 = FunctionInput() {bUsePreviewValueAsDefault=true, inputName=Rotation Speed, inputType=FunctionInput_Scalar, sortPriority=20, previewValue=<0.02 0 0 1>}
  n21 = Multiply(A=Time(), B=@n20)
  n22 = Multiply(A=@n21) {constB=-1}
  n23 = MaterialFunctionCall(UVs=@n16, Rotation Center=@n17, "Rotation Angle (0-1)"=Add(A=@n19.R, B=If(A=@n19.R, "A > B"=@n21, "A < B"=@n22) {constB=0.5})) {materialFunction=&CustomRotator}
  n24 = FunctionInput() {bUsePreviewValueAsDefault=true, inputName=Aspect, inputType=FunctionInput_Scalar, previewValue=<1.5 0 0 1>}
  n25 = Add(A=Multiply(A=Subtract(A=@n23.Rotated Values, B=@n17) {constB=0.5}, B=Divide(B=@n24)), B=@n17)
  n26 = FunctionInput(Preview=TextureObject()) {
   bUsePreviewValueAsDefault=true
   description="Texture to manipulate (optional)"
   inputName=Texture To Warp
   inputType=FunctionInput_Texture2D
   sortPriority=30
  }
  FunctionOutput(None=TextureSample(UVs=FunctionOutput(None=@n25) {outputName=TexCoord Out}, Tex=@n26) {texture=/Engine/EngineResources/DefaultTexture}.RGB) {outputName=Warped Texture, sortPriority=10}
 ]
}
