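# Unlit Papaya material for glass shards and video screens that animate with event audio.
# Vertex blue blends crack-and-cubemap shard emissive into screen content, which static switches select from video, intermission glow or a redirect map.
# Video supports kaleidoscope, quadrant, mirror and split UVs, with cracked edge vignettes.
# Pivot Painter data drives per-shard rotation and audio-driven push through world position offset, while Forward and Mobile paths use simplified versions.
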Material /Game/Environments/Papaya/Materials/M_Shard_Papaya
graph {
 properties {bUseMaterialAttributes=true, pixelDepthOffsetMode=PDOM_Legacy, shadingModel=MSM_Unlit, usedShadingModels=Unlit}
 nodes [
  n1 = ShadingPathSwitch(Default=Fresnel() {baseReflectFraction=0.2, exponent=4}, Mobile=Constant() {r=0.2}) {desc=Migrated from FeatureLevelSwitch}
  n2 = ComponentMask(None=TransformPosition(None=WorldPosition().XYZ) {transformSourceType=TRANSFORMPOSSOURCE_World}) {b=true, r=true}
  n3 = ScalarParameter() {defaultValue=1000, parameterName=SizeShimmer}
  n4 = Time()
  n5 = LinearInterpolate(Alpha=VertexColor().R) {constA=0.7, constB=1.2}
  n6 = TextureSample(UVs=Panner(Coordinate=Divide(A=@n2, B=@n3), Time=Multiply(A=@n4, B=@n5)) {speedX=0.1, speedY=0.3}) {texture=&T_Inky_Smoke_Tile_Chromatic}
  n7 = ComponentMask(None=@n6.RGB) {g=true, r=true}
  n8 = ShadingPathSwitch(Default=Multiply(A=Multiply(A=@n7, B=@n7), B=Constant2Vector() {g=0.05, r=0.05}), Mobile=Constant()) {desc=Migrated from FeatureLevelSwitch}
  n9 = TextureCoordinate() {coordinateIndex=1}
  n10 = Multiply(A=TextureSample(UVs=Add(A=@n8, B=@n9)) {texture=/Game/Effects/Fort_Effects/Textures/Tile/T_Glass_Cracks_M}.G) {constB=0.1}
  n11 = ComponentMask(None=TextureCoordinate() {coordinateIndex=1}) {g=true}
  n12 = Add(
   A=Add(A=TextureSample() {texture=&T_Glass_Cracks_M}.G, B=@n10)
   B=Multiply(A=Power(Base=@n11) {constExponent=6.4}) {constB=0.15}
  )
  n13 = LinearInterpolate(Alpha=@n12) {constA=0.22}
  Add(A=@n1, B=@n13)
  Constant3Vector()
  n14 = Multiply(A=Time(), B=ScalarParameter() {defaultValue=1, parameterName=DebugTestSpeed})
  Divide(A=Add(A=Sine(None=@n14)))
  n15 = Constant2Vector() {g=0.5, r=0.5}
  n16 = Constant2Vector() {g=1, r=0.5625}
  n17 = TextureCoordinate() {coordinateIndex=1}
  n18 = If(
   A=CollectionParameter() {collection=&/MPC_PapayaGadgetControls, parameterName=ScreenKaleidoscope}
   "A > B"=MaterialFunctionCall!(Aspect, Fracture ID Texture, TexCoord=@n17, Rotation Pivot, Rotation Speed, Texture To Warp) {materialFunction=/Game/Environments/Papaya/Materials/MaterialFunctions/MF_Papaya_Kaleidoscope}.TexCoord Out
   "A == B"=@n17
   "A < B"=@n17
  ) {constB=0.5}
  n19 = If(
   A=CollectionParameter() {collection=&MPC_PapayaGadgetControls, parameterName=ScreensUseQuadrants}
   "A > B"=Add(A=Multiply(A=@n17, B=@n16), B=Constant2Vector() {r=0.2188})
   "A == B"=@n18
   "A < B"=@n18
  ) {constB=0.5}
  n20 = ScalarParameter() {defaultValue=1, parameterName=VignetteBlend}
  n21 = ComponentMask(None=@n19) {r=true}
  n22 = Multiply(A=OneMinus(None=@n21), B=@n21)
  n23 = ComponentMask(None=@n19) {g=true}
  n24 = Multiply(A=OneMinus(None=@n23), B=@n23)
  n25 = Max(A=OneMinus(None=Multiply(A=@n22) {constB=2}), B=OneMinus(None=Multiply(A=@n24) {constB=2}))
  n26 = Power(Base=@n25, Exp=ScalarParameter() {defaultValue=0.729, parameterName=ContrastSquareVignette})
  n27 = ScalarParameter() {defaultValue=0.9898, parameterName=TransitionSquareVignette}
  n28 = TextureCoordinate()
  n29 = ScalarParameter() {defaultValue=1.0836, parameterName=EdgeShardSize}
  n30 = TextureSample(UVs=Rotator(Coordinate=Multiply(A=@n28, B=@n29), Time=ScalarParameter() {defaultValue=51.5018, parameterName=EdgeShardRotation})) {texture=&T_Glass_Cracks_M}
  n31 = MaterialFunctionCall(
   Transition Phase=Add(A=Multiply(A=@n26, B=@n27), B=ScalarParameter() {defaultValue=-0.5, parameterName=offsetSquareVignette})
   Height Texture=@n30.R
   Contrast=ScalarParameter() {parameterName=EdgeContrast}
  ) {materialFunction=&HeightLerp}
  n32 = Add(A=@n19, B=Multiply(A=Subtract(A=@n15, B=@n19), B=Multiply(A=@n20, B=@n31.Alpha)))
  n33 = ShadingPathSwitch(Default=@n32, Deferred=@n32, Forward=@n19, Mobile=@n19)
  n34 = MaterialFunctionCall(Float2=@n33) {materialFunction=&BreakOutFloat2Components}
  n35 = If(
   A=ScalarParameter() {group=Side Screen Display Options, parameterName=Mirror Media, sortPriority=0}
   "A > B"=MaterialFunctionCall(X=OneMinus(None=@n34.R), Y=@n34.G) {materialFunction=&MakeFloat2}.Result
   "A < B"=@n33
  ) {constB=0.5}
  n36 = Multiply(A=@n35, B=Constant2Vector() {g=1, r=0.5})
  n37 = Constant2Vector() {r=0.5}
  n38 = ScalarParameter() {group=Side Screen Display Options, parameterName=Split Media Left, sortPriority=10}
  n39 = MaterialFunctionCall!(
   UVs=If(
    A=ScalarParameter() {group=Side Screen Display Options, parameterName=Split Media Right, sortPriority=20}
    "A > B"=Add(A=@n36, B=@n37)
    "A < B"=If(A=@n38, "A > B"=@n36, "A < B"=@n35) {constB=0.5}
   ) {constB=0.5}
   Texture=StaticSwitch(
    True=TextureObject() {texture=/Game/Movies/Season5/Papaya_MediaPlayer_Video3}
    False=TextureObject() {texture=/Game/Movies/Season5/Papaya_MediaPlayer_Video2}
    Value=StaticBoolParameter() {parameterName=Screen3}
   )
  ) {materialFunction=/Game/Movies/InversePostProcess/MF_invertFNtonemap}
  n40 = Multiply(
   A=Power(Base=@n25, Exp=ScalarParameter() {defaultValue=0.4442, parameterName=ContrastSquareVignette2})
   B=ScalarParameter() {defaultValue=1.0113, parameterName=TransitionSquareVignette2}
  )
  n41 = ScalarParameter() {defaultValue=-0.476, parameterName=offsetSquareVignette2}
  n42 = CollectionParameter() {collection=&MPC_PapayaGadgetControls, parameterName=ScreenKaleidoscope}
  n43 = TextureCoordinate() {coordinateIndex=1}
  n44 = TextureSample(UVs=@n43) {samplerType=SAMPLERTYPE_LinearGrayscale, texture=/Game/Environments/Papaya/Textures/shatter_kaleidoscope_id}
  n45 = Add(
   A=@n30.G
   B=Add(
    A=TextureSample(UVs=@n43) {samplerType=SAMPLERTYPE_LinearGrayscale, texture=/Game/Environments/Papaya/Textures/shatter_kaleidoscope_lines}.R
    B=Multiply(A=@n44.R) {constB=0.5}
   )
  )
  n46 = Multiply(
   A=MaterialFunctionCall(
    Transition Phase=Add(A=@n40, B=@n41)
    Height Texture=If(A=@n42, "A > B"=@n45, "A < B"=@n30.G) {constB=0.5}
    Contrast=ScalarParameter() {parameterName=EdgeContrast2}
   ) {materialFunction=&HeightLerp}.Alpha
   B=ScalarParameter() {defaultValue=0.15, parameterName=FadeinShards}
  )
  n47 = Add(A=@n39.Result, B=@n46)
  n48 = ShadingPathSwitch(Default=@n47, Deferred=@n47, Forward=@n39.Result, Mobile=@n39.Result)
  n49 = Constant() {r=1}
  n50 = VectorParameter() {parameterName=ScaleUV2, defaultValue=<rgba 1 1.8 0 1>}
  n51 = VectorParameter() {parameterName=ScaleUV2x, defaultValue=<rgba 1.09 1.89 0 1>}
  n52 = CollectionParameter() {collection=&FortniteMaterialParameters, parameterName=EventAudio_Spectrum}
  n53 = MaterialFunctionCall(Float4=@n52) {materialFunction=&BreakOutFloat4Components}
  n54 = MaterialFunctionCall(
   UVs=TextureCoordinate() {coordinateIndex=1}
   Texture Scale=ComponentMask(None=LinearInterpolate(A=@n50.RGB, B=@n51.RGB, Alpha=@n53.R)) {g=true, r=true}
  ) {materialFunction=&ScaleUVsByCenter}
  n55 = TransformPosition(None=CameraPositionWS()) {transformSourceType=TRANSFORMPOSSOURCE_World}
  n56 = VectorParameter() {parameterName=LocalCenter, defaultValue=<rgba 0 0 500 1>}
  n57 = Subtract(A=@n55, B=@n56.RGB)
  n58 = Multiply(A=Divide(A=ComponentMask(None=@n57) {b=true, r=true}, B=Distance(A=@n55, B=@n56.RGB)), B=Constant2Vector() {g=-1, r=1})
  n59 = Add(A=@n54.UVs, B=Divide(A=@n58, B=ScalarParameter() {defaultValue=-2, parameterName=UVFakeParalax}))
  n60 = ScalarParameter() {defaultValue=-8, parameterName=UVFakeParalax2}
  n61 = Multiply(A=Add(A=@n54.UVs, B=Divide(A=@n58, B=@n60)), B=ScalarParameter() {defaultValue=2, parameterName=IntermissionShardScale})
  n62 = Time()
  n63 = ScalarParameter() {defaultValue=-4, parameterName=UVFakeParalax3}
  n64 = Multiply(A=Add(A=@n54.UVs, B=Divide(A=@n58, B=@n63)), B=ScalarParameter() {defaultValue=3, parameterName=IntermissionShardScale2})
  n65 = Time()
  n66 = Add(
   A=TextureSample(UVs=Rotator(Coordinate=@n61, Time=@n62) {centerX=0.9, centerY=1.1, speed=0.125}) {texture=&T_Glass_Cracks_M}.R
   B=TextureSample(UVs=Rotator(Coordinate=@n64, Time=@n65) {centerX=1.3, centerY=1.23, speed=-0.113}) {texture=&T_Glass_Cracks_M}.R
  )
  n67 = TextureSample(UVs=@n54.UVs) {texture=&T_Glass_Cracks_M}
  n68 = Multiply(A=Subtract(A=@n54.UVs, B=Constant2Vector() {g=0.5, r=0.5}), B=ShadingPathSwitch(Default=@n66, Deferred=@n66, Forward=@n67.R, Mobile=@n67.R))
  n69 = ScalarParameter() {defaultValue=-0.35, parameterName=ShardDistort}
  n70 = Distance(
   A=Add(A=ShadingPathSwitch(Default=@n59, Deferred=@n59, Forward=@n54.UVs, Mobile=@n54.UVs), B=Multiply(A=@n68, B=@n69))
   B=Constant2Vector() {g=0.5, r=0.5}
  )
  n71 = Divide(
   A=@n70
   B=LinearInterpolate(
    A=ScalarParameter() {defaultValue=0.03, parameterName=GlowRadius3}
    B=ScalarParameter() {defaultValue=0.035, parameterName=GlowRadius3x}
    Alpha=@n53.B
   )
  )
  n72 = Multiply(A=@n71, B=@n71)
  n73 = Desaturation(
   None=Multiply(
    A=MaterialFunctionCall!(
     Media Player TexCoord
     Media Color Code Coordinate
     Media Texture Resolution
     Base Color Parameter
     Base Color Mix
     Media Color Code Mix
     Intermission Color Parameter
     Intermission Color Mix
     Desaturation
    ) {materialFunction=/Game/Environments/Papaya/Materials/MaterialFunctions/MF_Papaya_GadgetColor}.Result
    B=ScalarParameter() {defaultValue=2, parameterName=Color Multiplier}
   )
   Fraction=ScalarParameter() {defaultValue=0.125, parameterName=Color Desat}
  )
  n74 = Multiply(A=@n73, B=Divide(A=OneMinus(None=@n70), B=ScalarParameter() {defaultValue=1.5, parameterName=GlowRadius2}))
  n75 = LinearInterpolate(
   A=ScalarParameter() {defaultValue=3.9276, parameterName=GlowFalloff2}
   B=ScalarParameter() {defaultValue=3, parameterName=GlowFalloff2x}
   Alpha=@n53.B
  )
  n76 = Multiply(A=@n73, B=ScalarParameter() {defaultValue=0.5, parameterName=OuterGlowBrightness})
  n77 = Power(
   Base=Divide(A=@n70, B=ScalarParameter() {defaultValue=1, parameterName=GlowRadius})
   Exp=ScalarParameter() {defaultValue=3, parameterName=GlowFalloff}
  )
  n78 = Add(
   A=Multiply(A=Divide(A=@n49, B=@n72), B=ScalarParameter() {defaultValue=1, parameterName=GlowBrightness3})
   B=Add(A=Power(Base=@n74, Exp=@n75), B=Multiply(A=@n76, B=@n77))
  )
  LinearInterpolate(A=@n48, B=@n78, Alpha=ScalarParameter() {parameterName=Intermission})
  MaterialFunctionCall(LDR Input=Saturate(None=@n73), HDR Tint=@n73, HDR Intensity Multiplier=ScalarParameter() {defaultValue=150, parameterName=HDR Intensity}) {materialFunction=&DeriveHDRfromLDR}
  n79 = MaterialFunctionCall(Float2=VectorParameter() {defaultValue=<rgba -1 1 0 1>}.RGB) {materialFunction=&BreakOutFloat2Components}
  n80 = Time()
  n81 = ScalarParameter() {parameterName=Speed Multiplier}
  n82 = CollectionParameter() {collection=&FortniteMaterialParameters, parameterName=EventAudio_Amplitude}
  n83 = ScalarParameter() {defaultValue=0.05, parameterName=MPC Speed Multiplier}
  n84 = Frac(None=Add(A=Multiply(A=@n80, B=@n81) {constB=0.05}, B=Multiply(A=@n82, B=@n83) {constB=0.05}))
  n85 = Add(
   A=Multiply(A=TextureCoordinate()) {constB=0.5}
   B=MaterialFunctionCall(R=Constant(), G=LinearInterpolate(A=@n79.R, B=@n79.R, Alpha=@n84) {constA=-1}) {materialFunction=&AppendMany}.RG
  )
  n86 = TextureSample(UVs=Rotator(Coordinate=@n85, Time=Constant() {r=6.2832})) {texture=/Game/Global/Textures/Default/Masks_and_Gradients/3pxRGBImage}
  Multiply(A=Desaturation(None=@n86.RGB, Fraction=ScalarParameter() {defaultValue=1, parameterName=Desaturation}), B=@n73)
  n87 = TextureSample(UVs=Multiply(A=@n85, B=Constant2Vector() {g=0.45, r=0.25})) {texture=/Game/Athena/Prototype/Blueprints/Sneezy/Textures/T_Sneezy_MagneticWaves}
  n88 = Constant2Vector() {g=0.7, r=0.5}
  n89 = TextureSample(UVs=Add(A=Multiply(A=@n85, B=@n88), B=Constant2Vector() {g=0.05}) {constB=0.05}) {texture=&T_Sneezy_MagneticWaves}
  n90 = Add(A=Multiply(A=@n85, B=Constant2Vector() {g=0.2, r=0.2}), B=Constant2Vector() {g=0.05}) {constB=0.05}
  Multiply(
   A=@n86.RGB
   B=LinearInterpolate(
    A=Add(A=@n87.RGB, B=@n89.RGB)
    B=TextureSample(UVs=@n90) {samplerType=SAMPLERTYPE_LinearColor, texture=&HighContrastNoise}.RGB
    Alpha=ScalarParameter() {parameterName=Mix Glow}
   )
  )
  n91 = TextureCoordinate() {coordinateIndex=1}
  n92 = VectorParameter() {parameterName=ScaleUV_RedirectArrow, defaultValue=<rgba 0.2 0.3429 0 1>}
  n93 = Add(
   A=ComponentMask(None=VectorParameter() {parameterName=OffsetUV_RedirectArrow, defaultValue=<rgba -0.2571 0 0 1>}.RGB) {g=true, r=true}
   B=MaterialFunctionCall(UVs=@n91, Texture Scale=ComponentMask(None=@n92.RGB) {g=true, r=true}) {materialFunction=&ScaleUVsByCenter}.UVs
  )
  n94 = Add(A=Multiply(A=Time(), B=ScalarParameter() {defaultValue=2, parameterName=SpeedArrowAnim}), B=ComponentMask(None=@n93) {r=true})
  n95 = ScalarParameter() {defaultValue=2.8544, parameterName=FrequencyArrowAnim}
  n96 = Sine(None=Frac(None=Divide(A=@n94, B=@n95)))
  Multiply(A=@n96)
  Reroute(None=@n10)
  n97 = Divide(A=Distance(A=WorldPosition().XYZ, B=ActorPositionWS()), B=ScalarParameter() {defaultValue=1490.7649, parameterName=Radial Distance})
  n98 = MaterialFunctionCall() {materialFunction=/Engine/Functions/Engine_MaterialFunctions02/PivotPainter/PivotPainter_PerObjectData}
  n99 = Time()
  n100 = Frac(None=Multiply(A=Add(A=@n98.Random Value Per Element, B=@n99), B=ScalarParameter() {defaultValue=-0.5, parameterName=Radial Speed}))
  n101 = Multiply(A=ScalarParameter() {defaultValue=0.05, parameterName=ReflectionDistort}, B=Sine(None=Add(A=@n97, B=@n100)))
  n102 = ReflectionVectorWS()
  n103 = Multiply(
   A=TextureSampleParameterCube!(UVs=Add(A=@n101, B=@n102), Apply View MipBias) {parameterName=Cubemap, samplerType=SAMPLERTYPE_LinearColor, texture=/Game/Packages/Fortress_Sky/TexturesHDR/T_AthenaSkylight}.RGB
   B=ComponentMask(None=MaterialFunctionCall() {materialFunction=/Engine/Functions/GetPostProcessFunctions/GetAmbientCubemapTint}.AmbientCubemapTint) {b=true, g=true, r=true}
  )
  n104 = Add(A=ShadingPathSwitch(Default=@n6.RGB, Mobile=Constant()) {desc=Migrated from FeatureLevelSwitch}, B=@n13)
  n105 = MaterialFunctionCall!(Use Alternate UVs, Alternate UVs) {materialFunction=/Game/Athena/Effects/Materials/Geode/Functions/MF_Geode_Crack_Colors}
  n106 = Add(
   A=Multiply(A=Multiply(A=@n104, B=@n1), B=Add(A=@n105.Cloud Color, B=@n73) {constA=0.2, constB=0.4})
   B=Multiply(A=Multiply(A=@n73) {constB=8}, B=@n12)
  )
  n107 = ParticleColor()
  n108 = Add(
   A=QualitySwitch(Default=Multiply(A=@n103, B=@n73), Low=Constant())
   B=Multiply(A=ScalarParameter() {defaultValue=0.1, parameterName=Emissive}, B=Multiply(A=@n106, B=@n107.RGB))
  )
  n109 = ComponentMask(None=VectorParameter() {parameterName=ScaleUV_Redirectmap, defaultValue=<rgba 1 1 0 1>}.RGB) {g=true, r=true}
  n110 = VectorParameter() {parameterName=OffsetUV_Redirectmap, defaultValue=<rgba 0.0095 0 0 1>}
  n111 = TextureSample(UVs=Add(
    A=MaterialFunctionCall(UVs=@n91, Texture Scale=@n109) {materialFunction=&ScaleUVsByCenter}.UVs
    B=ComponentMask(None=@n110.RGB) {g=true, r=true}
   )) {texture=/Game/Environments/Papaya/Textures/RedirectMap}
  n112 = ComponentMask(None=VectorParameter() {parameterName=ScaleUV_RedirectTarget, defaultValue=<rgba 0.432 0.9048 0 1>}.RGB) {g=true, r=true}
  n113 = VectorParameter() {parameterName=OffsetUV_RedirectTarget, defaultValue=<rgba 0.2286 0.0381 0 1>}
  n114 = TextureSample(UVs=Add(
    A=MaterialFunctionCall(UVs=@n91, Texture Scale=@n112) {materialFunction=&ScaleUVsByCenter}.UVs
    B=ComponentMask(None=@n113.RGB) {g=true, r=true}
   )) {texture=/Game/Environments/Papaya/Textures/RedirectMapTarget}
  n115 = TextureSample(UVs=@n93) {texture=/Game/Environments/Papaya/Textures/RedirectMapArrow}
  n116 = Ceil(None=@n96)
  n117 = StaticSwitchParameter(
   True=LinearInterpolate(A=LinearInterpolate(A=@n111.RGB, B=@n114.RGB, Alpha=@n114.RGB), B=@n115.RGB, Alpha=Multiply(A=@n116, B=@n115.A))
   False=StaticSwitchParameter(True=@n78, False=@n48) {parameterName=Is_Intermission_Screen}
  ) {parameterName=IsRedirectMap}
  n118 = VertexColor()
  n119 = Saturate(None=Subtract(A=@n118.B, B=@n46))
  n120 = MaterialFunctionCall() {materialFunction=&PivotPainter_PerObjectData}
  n121 = Multiply(A=Time(), B=ScalarParameter() {defaultValue=0.1, parameterName=SpeedRandomVector})
  n122 = Sine(None=@n121)
  n123 = Cosine(None=@n121)
  n124 = Sine(None=Add(A=@n121) {constB=0.33})
  n125 = Normalize(VectorInput=CrossProduct(
    A=@n120.X-Axis Vector
    B=MaterialFunctionCall(X=@n122, Y=@n123, Z=@n124) {materialFunction=&MakeFloat3}.Result
   ))
  n126 = Time()
  n127 = LinearInterpolate(
   A=ScalarParameter() {defaultValue=0.2, parameterName=SpeedRotation_1}
   B=ScalarParameter() {defaultValue=0.5, parameterName=SpeedRotation_2}
   Alpha=@n120.Random Value Per Element
  )
  n128 = Multiply(A=Sine(None=Multiply(A=@n126, B=@n127)), B=ScalarParameter() {defaultValue=0.015, parameterName=MaxRotation})
  n129 = WorldPosition()
  n130 = TransformPosition(None=VectorParameter() {parameterName=CenterPos, defaultValue=<rgba 0 0 -100 1>}.RGB) {transformType=TRANSFORMPOSSOURCE_World}
  n131 = Normalize(VectorInput=Subtract(A=@n120.Pivot Position, B=@n130))
  n132 = CollectionParameter() {collection=&FortniteMaterialParameters, parameterName=EventAudio_Spectrum}
  n133 = Subtract(A=MaterialFunctionCall(Float4=@n132) {materialFunction=&BreakOutFloat4Components}.B) {constB=0.5}
  n134 = Multiply(
   A=Subtract(A=@n120.Random Value Per Element) {constB=0.5}
   B=Multiply(A=Multiply(A=@n133) {constB=2}, B=ScalarParameter() {defaultValue=0.05, parameterName=MaxSpin})
  )
  n135 = WorldPosition()
  n136 = CollectionParameter() {collection=&FortniteMaterialParameters, parameterName=EventAudio_Amplitude}
  n137 = ScalarParameter() {defaultValue=100, parameterName=OffsetDistancePushMax}
  n138 = LinearInterpolate(Alpha=@n120.Random Value Per Element) {constA=0.5}
  n139 = Power(
   Base=Divide(A=Distance(A=@n130, B=@n120.Pivot Position), B=ScalarParameter() {defaultValue=1500, parameterName=DistanceFadeWPOpush})
   Exp=ScalarParameter() {defaultValue=3, parameterName=DistanceFalloffWPOpush}
  )
  n140 = Multiply(A=@n136, B=Multiply(A=Multiply(A=@n137, B=@n138), B=Multiply(A=@n131, B=@n139)))
  n141 = Add(
   A=Add(
    A=RotateAboutAxis(NormalizedRotationAxis=@n125, RotationAngle=@n128, PivotPoint=@n120.Pivot Position, Position=@n129.XYZ)
    B=RotateAboutAxis(NormalizedRotationAxis=@n131, RotationAngle=@n134, PivotPoint=@n120.Pivot Position, Position=@n135.XYZ)
   )
   B=@n140
  )
  n142 = ShadingPathSwitch(Default=@n141, Deferred=@n141, Forward=@n140, Mobile=@n140)
  n143 = ScalarParameter() {defaultValue=0.05, parameterName=ScreenWPO}
  SetMaterialAttributes!(
   MaterialAttributes=BlendMaterialAttributes(
    A=SetMaterialAttributes!(MaterialAttributes, Emissive Color=@n108) {attributeSetTypes [#1]}
    B=SetMaterialAttributes(Emissive Color=@n117) {attributeSetTypes [#1]}
    Alpha=ShadingPathSwitch(Default=@n119, Deferred=@n119, Forward=@n118.B, Mobile=@n118.B)
   )
   World Position Offset=StaticSwitchParameter(True=@n142, False=Multiply(A=@n142, B=@n143)) {parameterName=IsShards}
  ) {attributeSetTypes [#2]}
  TextureCoordinate() {coordinateIndex=1, uTiling=2, vTiling=1.2}
  TextureCoordinate() {coordinateIndex=1}
  TextureSampleParameter2D() {
   constCoordinate=1
   parameterName=Temp_Screen_Content
   texture=/Game/UI/Foundation/Textures/LoadingScreens/PSA/VBuckScamWarning/BR04_LoadingScreen_VBucksScam
  }
  VectorParameter() {parameterName=CubemapTint, defaultValue=<rgba 0.5611 0.1402 0.625 1>}
  VectorParameter() {parameterName=GlowColor, defaultValue=<rgba 1.1437 0.4875 1.8 1>}
  VectorParameter() {parameterName=GlowColor2, defaultValue=<rgba 0.1499 0 0.526 1>}
 ]
}
