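# Lit material for glowing neon sign tubes, usable on Nanite and instanced static meshes.
# Two adjustable power lobes of the normal-to-camera dot product sum into a view-facing glow, scaled by a time-of-day function.
# UseMusicColor blends a modulator from soft noise-driven flicker to the powered red channel of the event audio spectrum collection parameter.
# The resulting brightness multiplies an emissive color parameter, cyan-blue by default.
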Material /Game/Environments/Papaya/Materials/M_SignNeonTubes
graph {
 properties {bUsedWithInstancedStaticMeshes=true, bUsedWithNanite=true, pixelDepthOffsetMode=PDOM_Legacy, usedShadingModels=Default Lit}
 nodes [
  Constant() {r=0.01}
  Constant() {r=1}
  n1 = Constant() {r=0.5}
  n2 = Constant() {r=0.2}
  n3 = Constant() {r=0.7}
  n4 = Constant() {r=1}
  n5 = DotProduct(A=PixelNormalWS(), B=CameraVectorWS())
  n6 = Multiply(
   A=Power(Base=@n5, Exp=ScalarParameter() {defaultValue=8, parameterName=Exponent2})
   B=ScalarParameter() {defaultValue=10, parameterName=Brightness2}
  )
  n7 = Multiply(
   A=Power(Base=@n5, Exp=ScalarParameter() {defaultValue=2, parameterName=Exponent})
   B=ScalarParameter() {defaultValue=20, parameterName=Brightness}
  )
  n8 = ScalarParameter() {defaultValue=0.5, parameterName=FlickerMinBrightness}
  n9 = TextureSample(UVs=Multiply(A=Time(), B=ScalarParameter() {defaultValue=0.5, parameterName=SoftFlickerSpeed})) {samplerType=SAMPLERTYPE_Grayscale, texture=&LowResBlurredNoise}
  n10 = CollectionParameter() {collection=&FortniteMaterialParameters, parameterName=EventAudio_Spectrum}
  n11 = Clamp(None=ComponentMask(None=@n10) {r=true})
  n12 = ScalarParameter() {defaultValue=3, parameterName=percExponent}
  n13 = Multiply(
   A=Multiply(
    A=MaterialFunctionCall(Value_Morning=@n1, Value_Day=@n2, Value_Evening=@n3, Value_Night=@n4) {materialFunction=&Scalar_Time_of_Day}.Result
    B=Add(A=@n6, B=@n7)
   )
   B=LinearInterpolate(A=LinearInterpolate(A=@n8, Alpha=@n9.R), B=Power(Base=@n11, Exp=@n12), Alpha=ScalarParameter() {parameterName=UseMusicColor})
  )
  Multiply(A=VectorParameter() {parameterName=EmissivColor, defaultValue=<rgba 0 0.4324 1 1>}.RGB, B=@n13)
  Multiply(A=@n13)
  ScalarParameter() {defaultValue=0.2, parameterName=Roughness}
 ]
}
