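# Material function, exposed to the library, that rasterizes texture coordinates into a grid.
# It multiplies the TexCoord input by Tiles XY, floors the result, and scales it back by the reciprocal of the tile count.
# A half-tile offset is then added, so every point within a tile outputs that tile's center coordinate.
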MaterialFunction /Game/Environments/Papaya/Materials/MF_GridTexCoord
graph {
 properties {bExposeToLibrary=true, description=Rasterize Texture Coordinates into a Grid}
 nodes [
  n1 = FunctionInput() {inputName=Tiles XY, inputType=FunctionInput_Vector2, sortPriority=10}
  n2 = Floor(None=Multiply(A=FunctionInput() {inputName=TexCoord, inputType=FunctionInput_Vector2}, B=@n1))
  n3 = Divide(A=Constant() {r=1}, B=@n1)
  FunctionOutput(None=Add(A=Multiply(A=@n2, B=@n3), B=Multiply(A=@n3) {constB=0.5}) {constB=0.05}) {outputName=TexCoord}
 ]
}
